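# Looping wind ambience for a player in a Skytube wind conduit, limited to four concurrent instances.
# An IsLocal switch selects a first-person conduit loop layered with border wind for the local player, or a third-person conduit loop with Skytube 3P attenuation for others.
# A Speed parameter drives continuous pitch and volume modulation on the whole cue and separately modulates the border wind layer.
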SoundCue /FNGameplayCues/Athena/Sounds/Ambient/Skytube/SC_Skytube_Loop_Player_S9
root=ModulatorContinuous_0
nodes [
 {
  id=Attenuation_0
  children [Mixer_1]
  properties {
   attenuationSettings=/Game/Athena/Sounds/Ambient/Skytube/Skytube_3P_Att
   attenuationOverrides {falloffDistance=10000, stereoSpread=512, attenuationShapeExtents {x=0}}
  }
 }
 {id=Attenuation_1, children [Mixer_0]}
 {id=Mixer_0, children [WavePlayer_0, ModulatorContinuous_1], properties {inputVolume [1,0.5]}}
 {id=Mixer_1, children [WavePlayer_2], properties {inputVolume [1.5]}}
 {
  id=ModulatorContinuous_0
  children [Modulator_0]
  properties {
   pitchModulationParams {default=0, maxOutput=1.3, minOutput=0.9, parameterName=Speed}
   volumeModulationParams {default=0, minOutput=0.25, parameterName=Speed}
  }
 }
 {
  id=ModulatorContinuous_1
  children [WavePlayer_3]
  properties {pitchModulationParams {default=0, minOutput=1, parameterName=Speed}, volumeModulationParams {default=0, minInput=0.5, parameterName=Speed}}
 }
 {id=Modulator_0, children [Switch_0]}
 {id=Switch_0, children [Attenuation_0, Attenuation_0, Attenuation_1], properties {intParameterName=IsLocal}}
 {id=WavePlayer_0, properties {bLooping=true, soundWaveAssetPtr=/FNGameplayCues/Athena/Sounds/Ambient/Skytube/Rev1/WindConduit_First_Loop_01}}
 {id=WavePlayer_2, properties {bLooping=true, soundWaveAssetPtr=/FNGameplayCues/Athena/Sounds/Ambient/Skytube/Rev1/WindConduit_Third_Loop_02}}
 {id=WavePlayer_3, properties {bLooping=true, soundWaveAssetPtr=/FNGameplayCues/Sounds/Fort_Ambient_Looping/Wind_Exterior/border_wind_05Loop}}
]
properties {
 bOverrideConcurrency=true
 maxDistance=2097152
 soundClassObject=&Fort_Ambient_Athena
 volumeMultiplier=0.4
 attenuationOverrides {falloffDistance=6000, stereoSpread=2000, attenuationShapeExtents {x=0}}
 concurrencyOverrides {maxCount=4}
}
