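# Looping wind sound cue for a vehicle in a Skytube wind conduit, on the Athena ambient sound class.
# IsLocal switches between a two-layer local mix and a third-person loop, whose attenuation IsVehicleEmpty selects.
# A continuous Speed parameter drives overall pitch and volume, and separately modulates the local border-wind layer, after a Modulator node adds variation.
# Concurrency override caps simultaneous instances.
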SoundCue /FNGameplayCues/Athena/Sounds/Ambient/Skytube/Skytube_Loop_Vehicle_Cue
root=ModulatorContinuous_1
nodes [
 {
  id=Attenuation_0
  children [Mixer_1]
  properties {
   attenuationSettings=&/Skytube_EmptyVehicle_3P_Att
   attenuationOverrides {falloffDistance=10000, stereoSpread=512, attenuationShapeExtents {x=0}}
  }
 }
 {
  id=Attenuation_2
  children [Mixer_1]
  properties {
   attenuationSettings=/Game/Athena/Sounds/Ambient/Skytube/Skytube_3P_Att
   attenuationOverrides {falloffDistance=10000, stereoSpread=512, attenuationShapeExtents {x=0}}
  }
 }
 {id=Attenuation_3, children [Mixer_0]}
 {id=Mixer_0, children [WavePlayer_2, ModulatorContinuous_2], properties {inputVolume [1,0.5]}}
 {id=Mixer_1, children [WavePlayer_3], properties {inputVolume [1.5]}}
 {
  id=ModulatorContinuous_1
  children [Modulator_1]
  properties {
   pitchModulationParams {default=0, maxOutput=1.3, minOutput=0.9, parameterName=Speed}
   volumeModulationParams {default=0, minOutput=0.25, parameterName=Speed}
  }
 }
 {
  id=ModulatorContinuous_2
  children [WavePlayer_4]
  properties {pitchModulationParams {default=0, minOutput=1, parameterName=Speed}, volumeModulationParams {default=0, minInput=0.5, parameterName=Speed}}
 }
 {id=Modulator_1, children [Switch_1]}
 {id=Switch_0, children [Attenuation_2, Attenuation_2, Attenuation_0], properties {intParameterName=IsVehicleEmpty}}
 {id=Switch_1, children [Switch_0, Switch_0, Attenuation_3], properties {intParameterName=IsLocal}}
 {id=WavePlayer_2, properties {bLooping=true, soundWaveAssetPtr=/FNGameplayCues/Athena/Sounds/Ambient/Skytube/Rev1/WindConduit_First_Loop_01}}
 {id=WavePlayer_3, properties {bLooping=true, soundWaveAssetPtr=/FNGameplayCues/Athena/Sounds/Ambient/Skytube/Rev1/WindConduit_Third_Loop_02}}
 {id=WavePlayer_4, properties {bLooping=true, soundWaveAssetPtr=/FNGameplayCues/Sounds/Fort_Ambient_Looping/Wind_Exterior/border_wind_05Loop}}
]
properties {
 bOverrideConcurrency=true
 maxDistance=2097152
 soundClassObject=&Fort_Ambient_Athena
 volumeMultiplier=0.832
 attenuationOverrides {falloffDistance=6000, stereoSpread=2000, attenuationShapeExtents {x=0}}
 concurrencyOverrides {maxCount=4}
}
