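# Sound cue for passing through a Skytube ring, playing wind conduit entry sounds in the Athena ambient sound class.
# An equally weighted random pick of four waves feeds a modulator for variation.
# A root continuous modulator then drives pitch and volume from the Intensity parameter, with a raised floor on each.
# Distance attenuation is off, and overridden concurrency caps simultaneous instances at three with volume scaling between them.
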SoundCue /FNGameplayCues/Athena/Sounds/Ambient/Skytube/Skytube_Ring_Passthrough_Cue
root=ModulatorContinuous_0
nodes [
 {
  id=ModulatorContinuous_0
  children [Modulator_0]
  properties {
   pitchModulationParams {default=0, minOutput=0.6, parameterName=Intensity}
   volumeModulationParams {default=0, minOutput=0.4, parameterName=Intensity}
  }
 }
 {id=Modulator_0, children [Random_2]}
 {id=Random_2, children [WavePlayer_8, WavePlayer_9, WavePlayer_11, WavePlayer_12], properties {weights [1,1,1,1]}}
 {id=WavePlayer_11, properties {soundWaveAssetPtr=/Game/Athena/Sounds/Ambient/Skytube/Rev1/WindConduit_Enter_06}}
 {id=WavePlayer_12, properties {soundWaveAssetPtr=/Game/Athena/Sounds/Ambient/Skytube/Rev1/WindConduit_Enter_07}}
 {id=WavePlayer_8, properties {soundWaveAssetPtr=/Game/Athena/Sounds/Ambient/Skytube/Rev1/WindConduit_Enter_01}}
 {id=WavePlayer_9, properties {soundWaveAssetPtr=/Game/Athena/Sounds/Ambient/Skytube/Rev1/WindConduit_Enter_03}}
]
properties {
 bOverrideConcurrency=true
 maxDistance=2097152
 soundClassObject=&Fort_Ambient_Athena
 volumeMultiplier=1.247
 attenuationOverrides {bAttenuate=false, falloffDistance=6000, stereoSpread=2000, attenuationShapeExtents {x=0}}
 concurrencyOverrides {maxCount=3, volumeScale=0.75, volumeScaleAttackTime=0.25}
}
